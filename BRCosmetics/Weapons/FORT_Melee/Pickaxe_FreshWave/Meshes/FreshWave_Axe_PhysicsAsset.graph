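# Physics asset for the FreshWave Axe pickaxe mesh, a two-handed melee cosmetic weapon.
# It holds a single capsule body on the Melee_TwoHanded_Handle bone, offset and rotated to roughly enclose the weapon.
# There are no constraints, so the whole mesh acts as one rigid collision shape.
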PhysicsAsset /BRCosmetics/Weapons/FORT_Melee/Pickaxe_FreshWave/Meshes/FreshWave_Axe_PhysicsAsset
bodies {Melee_TwoHanded_Handle [{shapeType=Capsule, length=69.8554, radius=40.8446, center=<3.1134 -1.0723 26.4919>, rotation=<rot 0.1415 -90 1.7711>}]}
constraints []
